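# Translucent, unlit wireframe material for Niagara mesh particles, stored with the Cloaked Assassin character's effects and named for its NPC variant.
# It renders meshes as two-sided polygon edges through the material's wireframe mode, with no textures.
# Particle color RGB feeds the shared MF_FX_EmissiveAdjustments function as emissive, so the particle system controls the line color.
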Material /Game/Characters/Player/Female/Medium/Bodies/F_MED_Cloaked_Assassin/Effects/M_PolyWireframe_CloakedAssassinNPC
graph {
 properties {
  bUsedWithNiagaraMeshParticles=true
  blendMode=BLEND_Translucent
  pixelDepthOffsetMode=PDOM_Legacy
  shadingModel=MSM_Unlit
  twoSided=true
  usedShadingModels=Unlit
  wireframe=true
 }
 nodes [MaterialFunctionCall(Emissive=ParticleColor().RGB) {materialFunction=&MF_FX_EmissiveAdjustments}]
}
